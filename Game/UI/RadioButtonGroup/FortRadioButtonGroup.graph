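# FortRadioButtonGroup is a reusable UI widget that shows a vertical, single-selection radio button list built from a data array.
# UpdateButtons rebuilds evenly filled buttons using per-item class, text and style, registers them in a new selection-required CommonUI legacy group, and applies optional size overrides.
# Clicks select that index and broadcast RadioButtonClicked_Event with it.
# CenterOnWidget centers on the first created button, and Destruct clears the list.
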WidgetBlueprint /Game/UI/RadioButtonGroup/FortRadioButtonGroup
graphs [
 CenterOnWidget {nodes [
   CenterOnWidget() {is=title K2Node_FunctionEntry}
   > CenterWidget!(
    execute
    ToCenter:Widget Object Reference="Get (a copy)"(Array:Array of Icon Text Button Object References=Get CreatedButtons() -> (CreatedButtons:Array of Icon Text Button Object References)) -> (Output:Icon Text Button Object Reference)
   ) -> (then)
  ]}
 EventGraph {nodes [
   UpdateButtons() {is=title K2Node_CustomEvent}
   > n1 = Sequence!(execute) -> (then_0, then_1, then_2:Exec, then_3:Exec)
   > ClearChildren(self=Get VerticalBox_Buttons() -> (VerticalBox_Buttons:Vertical Box Object Reference))
   > Clear(TargetArray:Array of Icon Text Button Object References=Get CreatedButtons())
   > n2 = Construct Object from Class!(
    execute
    Class=Get CommonButtonGroup() -> (CommonButtonGroup:Common Button Group Legacy Class Reference)
    self=Get VerticalBox_Buttons()
    bSelectionRequired:bool=true
   ) -> (then, ReturnValue:Common Button Group Legacy Object Reference)
   > n3 = Set CreatedButtonsGroup(CreatedButtonsGroup:Common Button Group Legacy Object Reference=@n2) -> (Output_Get:Common Button Group Legacy Object Reference)
   > Bind Event to On Button Clicked(self:Common Button Group Legacy Object Reference=@n3, Delegate=Create Event()) {is=title K2Node_AddDelegate}
   n4 = For Each Loop(
    Exec=@n1.then_1
    Array:Array of Fort Radio Button Item Struct Structures=Get Buttons() -> (Buttons:Array of Fort Radio Button Item Struct Structures)
   ) -> (Array Element:Fort Radio Button Item Struct Structure)
   n5 = Break Fort Radio Button Item Struct!(FortRadioButtonItemStruct:"Fort Radio Button Item Struct Structure (by ref)"=@n4) -> (
    ButtonClass_16_#1:Icon Text Button Class Reference
    ButtonStyle_12_#2:Common Button Style Class Reference
    ButtonText_8_#3:Text
   ) {is=title K2Node_BreakStruct}
   > n6 = Create Widget!(
    execute:Exec
    Class:User Widget Class Reference=@n5
    OwningPlayer:Player Controller Object Reference
    ButtonText:Text=@n5.ButtonText_8_#3
    IconBrush:Slate Brush Structure
    ControllerInputStyle:Common Button Style Class Reference
    ButtonTextJustification:ETextJustify Enum=Center
    bDisplayAllCaps:bool=True
    Font:Slate Font Info Structure=Get Class Defaults!(Class:Object Class Reference=@n5) -> (
     ButtonText:Text
     IconBrush:Slate Brush Structure
     ControllerInputStyle:Common Button Style Class Reference
     MouseKeyboardStyle:Common Button Style Class Reference
     ContentAlignment:EHorizontalAlignment Enum
     bMouseKeyboardStyleSet:bool=~
     OverrideButtonText:Text
     ButtonTextJustification:ETextJustify Enum
     bDisplayAllCaps:bool=~
     PressProgress:double=~
     bIgnoreInputActionWidgetText:bool=~
     FocusChangedEvent:Multicast Delegate
     Font:Slate Font Info Structure
     UseLightInput:bool=~
    ) {is=title K2Node_GetClassDefaults}.Font
    Style:Common Button Style Class Reference=@n5.ButtonStyle_12_#2
    bApplyAlphaOnDisable:bool=False
    bLocked:bool=False
    bSelectable:bool=true
    bShouldSelectUponReceivingFocus:bool=False
    bInteractableWhenSelected:bool=true
    bToggleable:bool=False
    bTriggerClickedAfterSelection:bool=False
    bDisplayInputActionWhenNotInteractable:bool=True
    bHideInputActionWithKeyboard:bool=False
    bShouldUseFallbackDefaultInputAction:bool=True
    bRequiresHold:bool=False
    HoldData:Common UIHold Data Class Reference
    ClickMethod:EButtonClickMethod Enum=DownAndUp
    InputPriority:int=0
    TriggeringInputAction:Data Table Row Handle Structure
   ) -> (then:Exec, ReturnValue:Icon Text Button Object Reference)
   > SetIsSelectable!(execute, self:Common Button Base Object Reference=@n6, bInIsSelectable:bool=true) -> (then)
   > SetIsInteractableWhenSelected!(execute, self:Common Button Base Object Reference=@n6, bInInteractableWhenSelected:bool=true) -> (then)
   > Add(TargetArray:Array of Icon Text Button Object References=Get CreatedButtons(), NewItem:"Icon Text Button Object Reference (by ref)"=@n6) {is=title K2Node_CallArrayFunction}
   > n7 = AddChild(self=Get VerticalBox_Buttons(), Content=@n6)
   > n8 = Cast To VerticalBoxSlot!(execute, Object=@n7) -> (then, CastFailed, AsVertical Box Slot:Vertical Box Slot Object Reference) {is=title K2Node_DynamicCast}
   > SetSize!(
    execute
    self:Vertical Box Slot Object Reference=@n8
    InSize:Slate Child Size Structure=Make Slate Child Size!(Value:float=1, SizeRule:ESlateSizeRule Enum=Fill) -> (SlateChildSize:Slate Child Size Structure) {is=title K2Node_MakeStruct}
   ) -> (then)
   > AddWidget!(
    execute
    self:Common Widget Group Base Object Reference=Get CreatedButtonsGroup() -> (CreatedButtonsGroup:Common Button Group Legacy Object Reference)
    InWidget:Widget Object Reference=@n6
   ) -> (then)
   n9 = Get WidthOverride() -> (WidthOverride:double)
   Branch(execute=@n1.then_2, Condition="float > float"(A=@n9, B=0) {is=title K2Node_CallFunction})
   > SetWidthOverride(self=Get SizeBox_Buttons() -> (SizeBox_Buttons:Size Box Object Reference), InWidthOverride=@n9)
   n10 = Get HeightOverride() -> (HeightOverride:double)
   Branch(execute=@n1.then_3, Condition="float > float"(A=@n10, B=0))
   > SetHeightOverride(self=Get SizeBox_Buttons(), InHeightOverride=@n10)
   n11 = OnButtonClicked_Bind!() -> (OutputDelegate, then, AssociatedButton:Common Button Legacy Object Reference, ButtonIndex:int) {is=title K2Node_CustomEvent}
   > SelectButtonAtIndex!(execute, self:Common Button Group Base Object Reference=Get CreatedButtonsGroup(), ButtonIndex:int=@n11.ButtonIndex, bAllowSound:bool=true) -> (then)
   > Call Radio Button Clicked Event!(execute:Exec, self:Fort Radio Button Group Object Reference, SelectedIndex:int=@n11.ButtonIndex) -> (then:Exec) {is=title K2Node_CallDelegate}
   Event Destruct()
   > ClearChildren(self=Get VerticalBox_Buttons())
  ]}
 RadioButtonClicked_Event {nodes [RadioButtonClicked_Event!() -> (then, SelectedIndex:int) {is=title K2Node_FunctionEntry}]}
]
variables [
 Buttons {type="array<FortRadioButtonItemStruct>", default []}
 CommonButtonGroup {type=Class, default=/Script/CommonUILegacy.CommonButtonGroupLegacy}
 CreatedButtons {type="array<IconTextButton_C>", default []}
 CreatedButtonsGroup {type=CommonButtonGroupLegacy, default=null}
 HeightOverride {type=number}
 RadioButtonClicked_Event {}
 WidthOverride {type=number}
]
widgets {
 info {inheritedWidgetCount=0, parentClass=&UserWidget, rootWidgetClass=&SizeBox, widgetCount=2}
 widgets [
  SizeBox_Buttons {widgetClassPath=&SizeBox, bIsVariable=true, widget=SizeBox_Buttons}
  VerticalBox_Buttons {widgetClassPath=&VerticalBox, bIsVariable=true, parent=SizeBox_Buttons, widget=VerticalBox_Buttons, slot {class=SizeBoxSlot}}
 ]
}
